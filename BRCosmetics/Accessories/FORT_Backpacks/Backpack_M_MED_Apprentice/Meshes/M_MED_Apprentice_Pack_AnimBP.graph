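# Animation Blueprint for the Apprentice backpack cosmetic, adding kill-reactive animation and springy crystal motion.
# Reactive variants blend in an Idle→OnKill→Outro state machine with additive idle, replacing the base idle.
# Enemy-elimination input events or a frontend timer trigger OnKill; delays and state notifies manage its flags.
# Chained spring controllers move root and crystal bones, with alphas zeroed during skydiving, jetpack or swimming, and eased out at high speed.
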AnimBlueprint /BRCosmetics/Accessories/FORT_Backpacks/Backpack_M_MED_Apprentice/Meshes/M_MED_Apprentice_Pack_AnimBP
target_skeleton=&/M_MED_Apprentice_Pack_Skeleton
graphs [
 AnimGraph {nodes [
   AnimGraphNode_StateMachine_3 = OnKillStateMachine() {is=title AnimGraphNode_StateMachine}
   n1 = Apply Additive(
    Base=@AnimGraphNode_StateMachine_3
    Additive="Sequence Player (additive) 'M_MED_Apprentice_Pack_Anim_Idle_REACTIVE'"() {sequence=/BRCosmetics/Accessories/FORT_Backpacks/Backpack_M_MED_Apprentice/Animation/M_MED_Apprentice_Pack_Anim_Idle_REACTIVE}
    Alpha=0.95
   ) {alphaBoolBlend {blendOption=Linear}}
   n2 = Sequence Player 'M_MED_Apprentice_Pack_Anim_Idle_BASE'() {sequence=/BRCosmetics/Accessories/FORT_Backpacks/Backpack_M_MED_Apprentice/Animation/M_MED_Apprentice_Pack_Anim_Idle_BASE}
   n3 = Get CosmeticVariantReactivity() -> (CosmeticVariantReactivity:bool)
   n4 = Spring controller - Bone: apprentice_pack_root!(
    ComponentPose:Component Space Pose Link Structure=Local To Component(LocalPose=Blend Poses by bool(BlendPose_0=@n1, BlendPose_1=@n2, bActiveValue=@n3) {blendType=Linear})
    Alpha:float=Get MainSpringAlpha() -> (MainSpringAlpha:double)
   ) -> (Pose:Component Space Pose Link Structure) {is=title AnimGraphNode_SpringBone, alphaBoolBlend {blendOption=Linear}, springBone {boneName=apprentice_pack_root}}
   n5 = Get SecondarySpringAlpha() -> (SecondarySpringAlpha:double)
   n6 = Spring controller - Bone: dyn_r_upper_crystal!(
    ComponentPose:Component Space Pose Link Structure=Spring controller - Bone: dyn_c_upper_crystal!(
     ComponentPose:Component Space Pose Link Structure=Spring controller - Bone: dyn_l_upper_crystal!(ComponentPose:Component Space Pose Link Structure=@n4, Alpha:float=@n5) -> (Pose:Component Space Pose Link Structure) {is=title AnimGraphNode_SpringBone, maxDisplacement=4, alphaBoolBlend {blendOption=Linear}, springBone {boneName=dyn_l_upper_crystal}}
     Alpha:float=@n5
    ) -> (Pose:Component Space Pose Link Structure) {is=title AnimGraphNode_SpringBone, maxDisplacement=4, alphaBoolBlend {blendOption=Linear}, springBone {boneName=dyn_c_upper_crystal}}
    Alpha:float=@n5
   ) -> (Pose:Component Space Pose Link Structure) {is=title AnimGraphNode_SpringBone, maxDisplacement=4, alphaBoolBlend {blendOption=Linear}, springBone {boneName=dyn_r_upper_crystal}}
   Output Pose(Result=Component To Local(ComponentPose=Spring controller - Bone: dyn_c_lower_crystal!(ComponentPose:Component Space Pose Link Structure=@n6, Alpha:float=@n5) -> (Pose:Component Space Pose Link Structure) {is=title AnimGraphNode_SpringBone, maxDisplacement=4, alphaBoolBlend {blendOption=Linear}, springBone {boneName=dyn_c_lower_crystal}})) {layerGroup=None}
   Save cached pose 'InPose'(Pose=Input Pose()) {cacheName=InPose}
  ]}
 OnKillStateMachine {
  entry=Idle
  path=AnimGraph.AnimGraphNode_StateMachine_3.OnKillStateMachine
  ids=id
  nodes [
   AnimStateNode_0 = Outro() {is=title AnimStateNode}
   AnimStateNode_1 = Idle() {stateEntered {guid=#1, notifyName=IdleStateEntered}}
   AnimStateNode_2 = OnKill() {is=title AnimStateNode, bAlwaysResetOnEntry=true, stateEntered {guid=#2, notifyName=OnKillStateEntered}}
   AnimStateTransitionNode_0 = AnimStateTransitionNode()
   AnimStateTransitionNode_1 = AnimStateTransitionNode() {bAutomaticRuleBasedOnSequencePlayerInState=true}
   AnimStateTransitionNode_2 = AnimStateTransitionNode() {crossfadeDuration=0}
  ]
  links [
   {
    dst=OnKill
    rule=AnimGraph.AnimGraphNode_StateMachine_3.OnKillStateMachine.AnimStateTransitionNode_2.Transition
    src=Idle
    via=AnimStateTransitionNode_2
   }
   {
    dst=Outro
    rule=AnimGraph.AnimGraphNode_StateMachine_3.OnKillStateMachine.AnimStateTransitionNode_0.Transition
    src=OnKill
    via=AnimStateTransitionNode_0
   }
   {
    dst=Idle
    rule=AnimGraph.AnimGraphNode_StateMachine_3.OnKillStateMachine.AnimStateTransitionNode_1.Transition
    src=Outro
    via=AnimStateTransitionNode_1
   }
  ]
 }
 Outro {
  path=AnimGraph.AnimGraphNode_StateMachine_3.OnKillStateMachine.AnimStateNode_0.Outro
  nodes [
   n1 = Sequence Player 'M_MED_Apprentice_Pack_Anim_Outro_REACTIVE'!(bLoopAnimation:bool=False) -> (Pose) {
    bLoopAnimation=false
    sequence=/BRCosmetics/Accessories/FORT_Backpacks/Backpack_M_MED_Apprentice/Animation/M_MED_Apprentice_Pack_Anim_Outro_REACTIVE
   }
   Output Animation Pose(Result=@n1)
  ]
 }
 Idle {
  path=AnimGraph.AnimGraphNode_StateMachine_3.OnKillStateMachine.AnimStateNode_1.Idle
  nodes [Output Animation Pose(Result="Sequence Evaluator (by time) 'M_MED_Apprentice_Pack_ReactAnim'"!(ExplicitTime:float=0) -> (Pose:Pose Link Structure) {
     is=title AnimGraphNode_SequenceEvaluator
     sequence=/BRCosmetics/Accessories/FORT_Backpacks/Backpack_M_MED_Apprentice/Animation/M_MED_Apprentice_Pack_ReactAnim
    })]
 }
 OnKill {
  path=AnimGraph.AnimGraphNode_StateMachine_3.OnKillStateMachine.AnimStateNode_2.OnKill
  nodes [Output Animation Pose(Result=Sequence Player!(
     Sequence:Anim Sequence Base Object Reference=Get KillAnimation() -> (KillAnimation:Anim Sequence Base Object Reference)
     bLoopAnimation:bool=False
    ) -> (Pose) {bLoopAnimation=false})]
 }
 Transition {
  path=AnimGraph.AnimGraphNode_StateMachine_3.OnKillStateMachine.AnimStateTransitionNode_0.Transition
  nodes [Result(bCanEnterTransition=Get activeKillAnimEnd() -> (activeKillAnimEnd:bool))]
 }
 Transition {path=AnimGraph.AnimGraphNode_StateMachine_3.OnKillStateMachine.AnimStateTransitionNode_1.Transition, nodes [Result()]}
 Transition {
  path=AnimGraph.AnimGraphNode_StateMachine_3.OnKillStateMachine.AnimStateTransitionNode_2.Transition
  nodes [Result(bCanEnterTransition=Get activeKill() -> (activeKill:bool))]
 }
 EventGraph {nodes [
   n1 = Event BlueprintUpdateAnimation()
   > n2 = Sequence!(execute) -> (then_0, then_1, then_2:Exec)
   > Randomly Play in Frontend!(Exec:Exec) -> (ExecOut:Exec) {is=title K2Node_MacroInstance}
   Branch(execute=@n2.then_1, Condition=Get activeKill())
   > n3 = Branch(Condition=IsValid(Object=Get KillAnimation()))
   n4 = Get KillAnimation()
   > Delay(Duration=float * float(A=Get SequenceLength!(self:Anim Sequence Base Object Reference=@n4) -> (SequenceLength:float), B=0.9) {is=title K2Node_PromotableOperator})
   > Set activeKill(activeKill:bool=false) -> (Output_Get:bool)
   Set activeKill(execute=@n3.else, activeKill=false)
   K2Node_Composite_0 = CalcSpringAlphaCutoff!(execute=@n2.then_2, deltaTime:double=@n1.DeltaTimeX) {is=title K2Node_Composite}
   Event OnCharacterCustomizationLoaded()
   > n5 = Cast To FortPlayerPawn(Object=TryGetFortPawn())
   > n6 = Cast To B_Athena_PartModifier_Backpack_Apprentice!(
    execute
    Object=GetModifierForPartType!(self:Fort Player Pawn Object Reference=@n5, PartType:EFortCustomPartType Enum=Backpack) -> (ReturnValue:Custom Character Part Modifier Object Reference)
   ) -> (then, CastFailed, AsB Athena Part Modifier Backpack Apprentice:B Athena Part Modifier Backpack Apprentice Object Reference) {is=title K2Node_DynamicCast}
   > Set B Athena Part Modifier Backpack Apprentice(B Athena Part Modifier Backpack Apprentice:B Athena Part Modifier Backpack Apprentice Object Reference=@n6) -> (Output_Get:B Athena Part Modifier Backpack Apprentice Object Reference)
   n7 = Event BlueprintProcessAnimInputEvent()
   > Branch(Condition=IsNamed(self=@n7.AnimInputEvent, InEventName=aie.EnemyPlayerEliminated))
   n8 = Get ReactivityEnabled!(self:B Athena Part Modifier Backpack Apprentice Object Reference=Get B Athena Part Modifier Backpack Apprentice() -> (B Athena Part Modifier Backpack Apprentice:B Athena Part Modifier Backpack Apprentice Object Reference)) -> (ReactivityEnabled:bool)
   > Set CosmeticVariantReactivity(CosmeticVariantReactivity:bool=@n8) -> (Output_Get:bool)
   > Set activeKill(activeKill=true)
   AnimNotify_OnKillStateEntered() {is=title K2Node_Event}
   n9 = Get KillAnimation()
   > RetriggerableDelay(Duration=float + float(
     A=GetPlayLength!(self:Animation Asset Object Reference=@n9) -> (ReturnValue:float)
     B=Get KillAnimationAdditionalTime() -> (KillAnimationAdditionalTime:double)
    ))
   > Set activeKillAnimEnd(activeKillAnimEnd:bool=true) -> (Output_Get:bool)
   AnimNotify_IdleStateEntered() {is=title K2Node_Event}
   > Set activeKillAnimEnd(activeKillAnimEnd=false)
   TryGetPawnOwner()
   OR Boolean(A=Get bIsInFrontEnd(), B=@n8)
  ]}
 CalcSpringAlphaCutoff {
  path=EventGraph.K2Node_Composite_0.CalcSpringAlphaCutoff
  nodes [
   n1 = Inputs!() -> (execute, deltaTime:double=~)
   > n2 = Branch(Condition=OR Boolean!(
     A=Get bIsSkyDiving()
     B=Get bIsParachuteOpened()
     C:bool=Get bIsUsingJetpack() -> (bIsUsingJetpack:bool)
     D:bool=Get bIsSurfaceSwimming()
     E:bool=Get bIsWaterJump() -> (bIsWaterJump:bool)
     F:bool=Get bIsSwimSprinting() -> (bIsSwimSprinting:bool)
    ) -> (ReturnValue))
   > Set MainSpringAlpha(MainSpringAlpha:double=0) -> (Output_Get:double)
   > Set SecondarySpringAlpha(SecondarySpringAlpha:double=0) -> (Output_Get:double)
   n3 = Set CurrentVelocity(execute=@n2.else, CurrentVelocity:double=Vector Length(A=Get PawnVelocity())) -> (Output_Get:double)
   n4 = "float > float"(A=@n3, B=1300) {is=title K2Node_CallFunction}
   n5 = float * float(A=@n1, B=20)
   > Set MainSpringAlpha(MainSpringAlpha=Lerp(A=Get MainSpringAlpha(), B=SelectFloat(A=0, B=0.035, bPickA=@n4), Alpha=@n5))
   > Set SecondarySpringAlpha(SecondarySpringAlpha=Lerp(A=Get SecondarySpringAlpha(), B=SelectFloat(A=0, B=0.005, bPickA=@n4), Alpha=@n5))
   Outputs()
  ]
 }
 RandomlyPlayInFrontend {nodes [
   Inputs!() -> (Exec:Exec)
   > Branch(Condition=Get bShouldDisplayFrontendReactivity() -> (bShouldDisplayFrontendReactivity:bool))
   > Set CosmeticVariantReactivity(CosmeticVariantReactivity=Get ReactivityEnabled(self=Get B Athena Part Modifier Backpack Apprentice()))
   > Delay(Duration=Get Timer() -> (Timer:double))
   > Set activeKill(activeKill=true)
   > Set Timer(Timer:double=10) -> (Output_Get:double)
   > Outputs!(ExecOut:Exec)
  ]}
]
variables [
 activeKill {type=boolean}
 activeKillAnimEnd {type=boolean}
 Alpha {type=number, default=0.005}
 B Athena Part Modifier Backpack Apprentice {type=B_Athena_PartModifier_Backpack_Apprentice_C, default=null}
 CosmeticVariantReactivity {type=boolean}
 CurrentVelocity {type=number}
 KillAnimation {type=AnimSequenceBase, default=&M_MED_Apprentice_Pack_ReactAnim}
 KillAnimationAdditionalTime {type=number, default=0.25}
 MainSpringAlpha {type=number}
 PlayReactive {type=boolean}
 SecondarySpringAlpha {type=number}
 skelMesh {type=SkeletalMeshComponent}
 Timer {type=number}
]
